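# FX material for the kelp blades of the FoundKelp glider cosmetic, with particle-driven color, dissolve, flicker and pattern UVs via dynamic parameters.
# Emissive layers a base color with a panned pattern multiplied by scrolling highlight lines built from frac gradients.
# Channel-packed masks, smoothstep dissolve with line highlight, and depth fade shape opacity.
# Multiplied sine and cosine waves pulse the green-mask glow, and texture-driven normal offset supplies optional WPO.
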Material /BRCosmetics/Gadgets/Assets/VinderTech_GliderChute/Glider_FoundKelp/FX/M_FoundKelp_Blades
graph {
 expressions {
  MaterialExpressionAdd_0 {
   class=MaterialExpressionAdd
   inputs [
    A {from=MaterialExpressionLinearInterpolate_4}
    B {from=MaterialExpressionAdd_1}
   ]
  }
  MaterialExpressionAdd_1 {
   class=MaterialExpressionAdd
   inputs [
    A {from=MaterialExpressionAdd_4}
    B {from=MaterialExpressionNamedRerouteUsage_0}
   ]
  }
  MaterialExpressionAdd_10 {
   class=MaterialExpressionAdd
   inputs [
    A {from=MaterialExpressionMultiply_46}
    B {from=MaterialExpressionMultiply_47}
   ]
  }
  MaterialExpressionAdd_11 {
   class=MaterialExpressionAdd
   inputs [
    A {from=MaterialExpressionSine_0}
    B {from=""}
   ]
  }
  MaterialExpressionAdd_12 {
   class=MaterialExpressionAdd
   inputs [
    A {from=MaterialExpressionCosine_0}
    B {from=""}
   ]
  }
  MaterialExpressionAdd_13 {
   class=MaterialExpressionAdd
   inputs [
    A {from=MaterialExpressionAdd_14}
    B {from=MaterialExpressionLinearInterpolate_0}
   ]
  }
  MaterialExpressionAdd_14 {
   class=MaterialExpressionAdd
   inputs [
    A {from=MaterialExpressionMultiply_12}
    B {from=MaterialExpressionSubtract_2}
   ]
   properties {constB=0}
  }
  MaterialExpressionAdd_16 {
   class=MaterialExpressionAdd
   inputs [
    A {from=MaterialExpressionSine_1}
    B {from=""}
   ]
  }
  MaterialExpressionAdd_17 {
   class=MaterialExpressionAdd
   inputs [
    A {from=MaterialExpressionVertexInterpolator_0, output=PS}
    B {from=MaterialExpressionMultiply_39}
   ]
  }
  MaterialExpressionAdd_18 {
   class=MaterialExpressionAdd
   inputs [
    A {from=MaterialExpressionMultiply_30}
    B {from=MaterialExpressionMultiply_20}
   ]
  }
  MaterialExpressionAdd_19 {
   class=MaterialExpressionAdd
   inputs [
    A {from=MaterialExpressionMultiply_19}
    B {from=MaterialExpressionMultiply_33}
   ]
  }
  MaterialExpressionAdd_2 {
   class=MaterialExpressionAdd
   inputs [
    A {from=MaterialExpressionMultiply_16}
    B {from=MaterialExpressionDynamicParameter_3, output=Dissolve Pattern}
   ]
  }
  MaterialExpressionAdd_20 {
   class=MaterialExpressionAdd
   inputs [
    A {from=MaterialExpressionMultiply_43}
    B {from=MaterialExpressionDynamicParameter_6, output=Pattern UV Add X}
   ]
  }
  MaterialExpressionAdd_21 {
   class=MaterialExpressionAdd
   inputs [
    A {from=MaterialExpressionMultiply_44}
    B {from=MaterialExpressionDynamicParameter_6, output=Pattern UV Add Y}
   ]
  }
  MaterialExpressionAdd_27 {
   class=MaterialExpressionAdd
   inputs [
    A {from=MaterialExpressionAdd_13}
    B {from=MaterialExpressionScalarParameter_11}
   ]
   properties {constB=0.05}
  }
  MaterialExpressionAdd_28 {
   class=MaterialExpressionAdd
   inputs [
    A {from=MaterialExpressionMultiply_9}
    B {from=MaterialExpressionAdd_8}
   ]
  }
  MaterialExpressionAdd_3 {
   class=MaterialExpressionAdd
   inputs [
    A {from=MaterialExpressionMultiply_6}
    B {from=MaterialExpressionMultiply_7}
   ]
  }
  MaterialExpressionAdd_4 {
   class=MaterialExpressionAdd
   inputs [
    A {from=MaterialExpressionMultiply_8}
    B {from=MaterialExpressionMultiply_1}
   ]
  }
  MaterialExpressionAdd_5 {
   class=MaterialExpressionAdd
   inputs [
    A {from=MaterialExpressionMultiply_21}
    B {from=MaterialExpressionAdd_1}
   ]
  }
  MaterialExpressionAdd_6 {
   class=MaterialExpressionAdd
   inputs [
    A {from=MaterialExpressionLinearInterpolate_13}
    B {from=MaterialExpressionMultiply_19}
   ]
  }
  MaterialExpressionAdd_7 {
   class=MaterialExpressionAdd
   inputs [
    A {from=MaterialExpressionMultiply_42}
    B {from=MaterialExpressionMultiply_45}
   ]
  }
  MaterialExpressionAdd_8 {
   class=MaterialExpressionAdd
   inputs [
    A {from=MaterialExpressionSmoothStep_0}
    B {from=MaterialExpressionAdd_18}
   ]
  }
  MaterialExpressionAdd_9 {
   class=MaterialExpressionAdd
   inputs [
    A {from=MaterialExpressionLinearInterpolate_8}
    B {from=MaterialExpressionMultiply_35}
   ]
  }
  MaterialExpressionAppendVector_0 {
   class=MaterialExpressionAppendVector
   inputs [
    A {from=MaterialExpressionVectorParameter_1, output=R}
    B {from=MaterialExpressionVectorParameter_1, output=R}
   ]
  }
  MaterialExpressionAppendVector_1 {
   class=MaterialExpressionAppendVector
   inputs [
    A {from=MaterialExpressionVectorParameter_1, output=B}
    B {from=MaterialExpressionDynamicParameter_1, output=Highlight Line Speed}
   ]
  }
  MaterialExpressionAppendVector_11 {
   class=MaterialExpressionAppendVector
   inputs [
    A {from=MaterialExpressionConstant4Vector_1, output=R}
    B {from=MaterialExpressionConstant4Vector_1, output=R}
   ]
  }
  MaterialExpressionAppendVector_12 {
   class=MaterialExpressionAppendVector
   inputs [
    A {from=MaterialExpressionConstant4Vector_0, output=B}
    B {from=MaterialExpressionConstant4Vector_0, output=B}
   ]
  }
  MaterialExpressionAppendVector_13 {
   class=MaterialExpressionAppendVector
   inputs [
    A {from=MaterialExpressionConstant4Vector_1, output=B}
    B {from=MaterialExpressionConstant4Vector_1, output=B}
   ]
  }
  MaterialExpressionAppendVector_2 {
   class=MaterialExpressionAppendVector
   inputs [
    A {from=MaterialExpressionVectorParameter_0, output=R}
    B {from=MaterialExpressionVectorParameter_0, output=R}
   ]
  }
  MaterialExpressionAppendVector_3 {
   class=MaterialExpressionAppendVector
   inputs [
    A {from=MaterialExpressionVectorParameter_0, output=B}
    B {from=MaterialExpressionVectorParameter_0, output=B}
   ]
  }
  MaterialExpressionAppendVector_4 {
   class=MaterialExpressionAppendVector
   inputs [
    A {from=MaterialExpressionVectorParameter_2, output=R}
    B {from=MaterialExpressionVectorParameter_2, output=R}
   ]
  }
  MaterialExpressionAppendVector_5 {
   class=MaterialExpressionAppendVector
   inputs [
    A {from=MaterialExpressionVectorParameter_2, output=B}
    B {from=MaterialExpressionVectorParameter_2, output=B}
   ]
  }
  MaterialExpressionAppendVector_6 {
   class=MaterialExpressionAppendVector
   inputs [
    A {from=MaterialExpressionScalarParameter_34, output=""}
    B {from=MaterialExpressionScalarParameter_35, output=""}
   ]
  }
  MaterialExpressionAppendVector_7 {
   class=MaterialExpressionAppendVector
   inputs [
    A {from=MaterialExpressionMultiply_37, output=""}
    B {from=MaterialExpressionMultiply_38, output=""}
   ]
  }
  MaterialExpressionAppendVector_8 {
   class=MaterialExpressionAppendVector
   inputs [
    A {from=MaterialExpressionAdd_20, output=""}
    B {from=MaterialExpressionAdd_21, output=""}
   ]
  }
  MaterialExpressionAppendVector_9 {
   class=MaterialExpressionAppendVector
   inputs [
    A {from=MaterialExpressionConstant4Vector_0, output=R}
    B {from=MaterialExpressionConstant4Vector_0, output=R}
   ]
  }
  MaterialExpressionComponentMask_1 {class=MaterialExpressionComponentMask, inputs [None {from=MaterialExpressionTextureCoordinate_1}], properties {g=true}}
  MaterialExpressionComponentMask_10 {class=MaterialExpressionComponentMask, inputs [None {from=MaterialExpressionTextureCoordinate_5}], properties {g=true}}
  MaterialExpressionComponentMask_11 {class=MaterialExpressionComponentMask, inputs [None {from=MaterialExpressionRotator_1}], properties {g=true}}
  MaterialExpressionComponentMask_12 {class=MaterialExpressionComponentMask, inputs [None {from=MaterialExpressionRotator_1}], properties {r=true}}
  MaterialExpressionComponentMask_13 {class=MaterialExpressionComponentMask, inputs [None {from=MaterialExpressionRotator_0}], properties {g=true}}
  MaterialExpressionComponentMask_14 {class=MaterialExpressionComponentMask, inputs [None {from=MaterialExpressionRotator_0}], properties {r=true}}
  MaterialExpressionComponentMask_2 {class=MaterialExpressionComponentMask, inputs [None {from=MaterialExpressionLocalPosition_0, output=XYZ}], properties {b=true, r=true}}
  MaterialExpressionComponentMask_3 {class=MaterialExpressionComponentMask, inputs [None {from=MaterialExpressionTextureCoordinate_2}], properties {g=true}}
  MaterialExpressionComponentMask_4 {class=MaterialExpressionComponentMask, inputs [None {from=MaterialExpressionTextureCoordinate_3}], properties {g=true}}
  MaterialExpressionComponentMask_5 {class=MaterialExpressionComponentMask, inputs [None {from=MaterialExpressionTextureCoordinate_6}], properties {g=true}}
  MaterialExpressionComponentMask_6 {class=MaterialExpressionComponentMask, inputs [None {from=MaterialExpressionTextureCoordinate_7}], properties {r=true}}
  MaterialExpressionComponentMask_7 {class=MaterialExpressionComponentMask, inputs [None {from=MaterialExpressionTextureCoordinate_7}], properties {g=true}}
  MaterialExpressionComponentMask_8 {class=MaterialExpressionComponentMask, inputs [None {from=MaterialExpressionTextureCoordinate_6}], properties {r=true}}
  MaterialExpressionComponentMask_9 {class=MaterialExpressionComponentMask, inputs [None {from=MaterialExpressionTextureCoordinate_5}], properties {r=true}}
  MaterialExpressionConstant4Vector_0 {class=MaterialExpressionConstant4Vector, inputs [], properties {constant=<rgba 1 1 0 0.25>}}
  MaterialExpressionConstant4Vector_1 {class=MaterialExpressionConstant4Vector, inputs [], properties {constant=<rgba 1 1 0 0.75>}}
  MaterialExpressionConvert_1 {
   class=MaterialExpressionConvert
   inputs [
    None {from=MaterialExpressionVectorParameter_9, output=R}
    None {from=MaterialExpressionVectorParameter_9, output=R}
   ]
   properties {
    convertInputs [
     {
      type=Scalar
      expressionInput {
       expression=&self:MaterialExpressionVectorParameter_9
       inputName=None
       mask=1
       maskA=0
       maskB=0
       maskG=0
       maskR=1
       outputIndex=1
      }
     }
     {
      type=Scalar
      expressionInput {
       expression=&self:MaterialExpressionVectorParameter_9
       inputName=None
       mask=1
       maskA=0
       maskB=0
       maskG=1
       maskR=0
       outputIndex=2
      }
     }
    ]
    convertOutputs [{type=Vector2}]
   }
  }
  MaterialExpressionConvert_2 {
   class=MaterialExpressionConvert
   inputs [
    None {from=MaterialExpressionVectorParameter_9, output=B}
    None {from=MaterialExpressionVectorParameter_9, output=B}
   ]
   properties {
    convertInputs [
     {
      type=Scalar
      expressionInput {
       expression=&self:MaterialExpressionVectorParameter_9
       inputName=None
       mask=1
       maskA=0
       maskB=1
       maskG=0
       maskR=0
       outputIndex=3
      }
     }
     {
      type=Scalar
      expressionInput {
       expression=&self:MaterialExpressionVectorParameter_9
       inputName=None
       mask=1
       maskA=1
       maskB=0
       maskG=0
       maskR=0
       outputIndex=4
      }
     }
    ]
    convertOutputs [{type=Vector2}]
   }
  }
  MaterialExpressionCosine_0 {class=MaterialExpressionCosine, inputs [None {from=MaterialExpressionTime_0, output=""}], properties {period=0.05}}
  MaterialExpressionDepthFade_0 {
   class=MaterialExpressionDepthFade
   inputs [
    Opacity {from=MaterialExpressionLinearInterpolate_23, output=""}
    FadeDistance {from=MaterialExpressionScalarParameter_15, output=""}
   ]
   properties {}
  }
  MaterialExpressionDivide_0 {
   class=MaterialExpressionDivide
   inputs [
    A {from=MaterialExpressionAdd_11}
    B {from=""}
   ]
  }
  MaterialExpressionDivide_1 {
   class=MaterialExpressionDivide
   inputs [
    A {from=MaterialExpressionAdd_12}
    B {from=""}
   ]
  }
  MaterialExpressionDivide_2 {
   class=MaterialExpressionDivide
   inputs [
    A {from=MaterialExpressionAdd_16}
    B {from=""}
   ]
  }
  MaterialExpressionDynamicParameter_0 {
   class=MaterialExpressionDynamicParameter
   inputs []
   properties {defaultValue=<rgba 1 0.5 0 0>, paramNames [" Dissolve Main", Highlight Line Speed, Dissolve Pattern, Dissolve Outline]}
  }
  MaterialExpressionDynamicParameter_1 {
   class=MaterialExpressionDynamicParameter
   inputs []
   properties {defaultValue=<rgba 1 0.5 0 0>, paramNames [" Dissolve Main", Highlight Line Speed, Dissolve Pattern, Dissolve Outline]}
  }
  MaterialExpressionDynamicParameter_2 {
   class=MaterialExpressionDynamicParameter
   inputs []
   properties {parameterIndex=1, defaultValue=<rgba 1 0 1 0>, paramNames [Color Change, Flicker, Mask_R_Opacity, 4]}
  }
  MaterialExpressionDynamicParameter_3 {
   class=MaterialExpressionDynamicParameter
   inputs []
   properties {defaultValue=<rgba 1 0.5 0 0>, paramNames [" Dissolve Main", Highlight Line Speed, Dissolve Pattern, Dissolve Outline]}
  }
  MaterialExpressionDynamicParameter_4 {
   class=MaterialExpressionDynamicParameter
   inputs []
   properties {parameterIndex=1, defaultValue=<rgba 1 0 1 0>, paramNames [Color Change, Flicker, Mask_R_Opacity, 4]}
  }
  MaterialExpressionDynamicParameter_5 {
   class=MaterialExpressionDynamicParameter
   inputs []
   properties {parameterIndex=1, defaultValue=<rgba 1 0 1 0>, paramNames [Color Change, Flicker, Mask_R_Opacity, 4]}
  }
  MaterialExpressionDynamicParameter_6 {
   class=MaterialExpressionDynamicParameter
   inputs []
   properties {parameterIndex=2, defaultValue=<rgba 1 1 1 0>, paramNames [Pattern UV Mult X, Pattern UV Mutl Y, Pattern UV Add X, Pattern UV Add Y]}
  }
  MaterialExpressionDynamicParameter_7 {
   class=MaterialExpressionDynamicParameter
   inputs []
   properties {defaultValue=<rgba 1 0.5 0 0>, paramNames [" Dissolve Main", Highlight Line Speed, Dissolve Pattern, Dissolve Outline]}
  }
  MaterialExpressionFrac_0 {class=MaterialExpressionFrac, inputs [None {from=MaterialExpressionMaterialFunctionCall_2, output=VGradient}], properties {}}
  MaterialExpressionFrac_2 {class=MaterialExpressionFrac, inputs [None {from=MaterialExpressionMaterialFunctionCall_4, output=VGradient}], properties {}}
  MaterialExpressionFrac_3 {class=MaterialExpressionFrac, inputs [None {from=MaterialExpressionMaterialFunctionCall_5, output=VGradient}], properties {}}
  MaterialExpressionLinearInterpolate_0 {
   class=MaterialExpressionLinearInterpolate
   inputs [
    A {from=""}
    B {from=""}
    Alpha {from=MaterialExpressionDynamicParameter_0, output=" Dissolve Main"}
   ]
   properties {constA=-0.2}
  }
  MaterialExpressionLinearInterpolate_1 {
   class=MaterialExpressionLinearInterpolate
   inputs [
    A {from=MaterialExpressionScalarParameter_13}
    B {from=MaterialExpressionScalarParameter_14}
    Alpha {from=MaterialExpressionMultiply_5}
   ]
   properties {constA=0.25, constB=0.75}
  }
  MaterialExpressionLinearInterpolate_10 {
   class=MaterialExpressionLinearInterpolate
   inputs [
    A {from=""}
    B {from=MaterialExpressionMultiply_31}
    Alpha {from=MaterialExpressionScalarParameter_40}
   ]
  }
  MaterialExpressionLinearInterpolate_11 {
   class=MaterialExpressionLinearInterpolate
   inputs [
    A {from=""}
    B {from=MaterialExpressionNamedRerouteUsage_1}
    Alpha {from=MaterialExpressionScalarParameter_41}
   ]
   properties {constA=1}
  }
  MaterialExpressionLinearInterpolate_12 {
   class=MaterialExpressionLinearInterpolate
   inputs [
    A {from=MaterialExpressionLinearInterpolate_14}
    B {from=MaterialExpressionMultiply_34}
    Alpha {from=MaterialExpressionScalarParameter_43}
   ]
  }
  MaterialExpressionLinearInterpolate_13 {
   class=MaterialExpressionLinearInterpolate
   inputs [
    A {from=MaterialExpressionTextureSampleParameter2D_0, output=R}
    B {from=MaterialExpressionMultiply_41}
    Alpha {from=MaterialExpressionScalarParameter_46}
   ]
  }
  MaterialExpressionLinearInterpolate_14 {
   class=MaterialExpressionLinearInterpolate
   inputs [
    A {from=MaterialExpressionLinearInterpolate_8}
    B {from=MaterialExpressionSaturate_1}
    Alpha {from=MaterialExpressionScalarParameter_47}
   ]
  }
  MaterialExpressionLinearInterpolate_15 {
   class=MaterialExpressionLinearInterpolate
   inputs [
    A {from=MaterialExpressionTextureCoordinate_5}
    B {from=MaterialExpressionAppendVector_8}
    Alpha {from=MaterialExpressionScalarParameter_48}
   ]
  }
  MaterialExpressionLinearInterpolate_16 {
   class=MaterialExpressionLinearInterpolate
   inputs [
    A {from=MaterialExpressionOneMinus_2}
    B {from=MaterialExpressionNamedRerouteUsage_7}
    Alpha {from=MaterialExpressionScalarParameter_42}
   ]
  }
  MaterialExpressionLinearInterpolate_17 {
   class=MaterialExpressionLinearInterpolate
   inputs [
    A {from=MaterialExpressionComponentMask_5}
    B {from=MaterialExpressionComponentMask_8}
    Alpha {from=MaterialExpressionScalarParameter_49}
   ]
  }
  MaterialExpressionLinearInterpolate_18 {
   class=MaterialExpressionLinearInterpolate
   inputs [
    A {from=MaterialExpressionVectorParameter_5, output=RGB}
    B {from=MaterialExpressionVectorParameter_8, output=RGB}
    Alpha {from=MaterialExpressionNamedRerouteUsage_8}
   ]
  }
  MaterialExpressionLinearInterpolate_19 {
   class=MaterialExpressionLinearInterpolate
   inputs [
    A {from=MaterialExpressionMultiply_4}
    B {from=MaterialExpressionLinearInterpolate_21}
    Alpha {from=MaterialExpressionScalarParameter_51}
   ]
  }
  MaterialExpressionLinearInterpolate_2 {
   class=MaterialExpressionLinearInterpolate
   inputs [
    A {from=""}
    B {from=""}
    Alpha {from=MaterialExpressionAdd_2}
   ]
   properties {constA=-1, constAlpha=1}
  }
  MaterialExpressionLinearInterpolate_20 {
   class=MaterialExpressionLinearInterpolate
   inputs [
    A {from=MaterialExpressionVectorParameter_5, output=RGB}
    B {from=MaterialExpressionLinearInterpolate_25}
    Alpha {from=MaterialExpressionScalarParameter_50}
   ]
  }
  MaterialExpressionLinearInterpolate_21 {
   class=MaterialExpressionLinearInterpolate
   inputs [
    A {from=MaterialExpressionMultiply_53}
    B {from=MaterialExpressionMultiply_56}
    Alpha {from=MaterialExpressionDynamicParameter_7, output=Highlight Line Speed}
   ]
  }
  MaterialExpressionLinearInterpolate_22 {
   class=MaterialExpressionLinearInterpolate
   inputs [
    A {from=MaterialExpressionLinearInterpolate_16}
    B {from=MaterialExpressionVertexColor_1, output=G}
    Alpha {from=MaterialExpressionScalarParameter_52}
   ]
  }
  MaterialExpressionLinearInterpolate_23 {
   class=MaterialExpressionLinearInterpolate
   inputs [
    A {from=MaterialExpressionLinearInterpolate_5}
    B {from=MaterialExpressionMultiply_18}
    Alpha {from=MaterialExpressionScalarParameter_53}
   ]
  }
  MaterialExpressionLinearInterpolate_24 {
   class=MaterialExpressionLinearInterpolate
   inputs [
    A {from=MaterialExpressionLinearInterpolate_7}
    B {from=MaterialExpressionMultiply_48}
    Alpha {from=MaterialExpressionScalarParameter_58}
   ]
  }
  MaterialExpressionLinearInterpolate_25 {
   class=MaterialExpressionLinearInterpolate
   inputs [
    A {from=MaterialExpressionLinearInterpolate_18}
    B {from=MaterialExpressionMultiply_49}
    Alpha {from=MaterialExpressionScalarParameter_59}
   ]
  }
  MaterialExpressionLinearInterpolate_26 {
   class=MaterialExpressionLinearInterpolate
   inputs [
    A {from=MaterialExpressionComponentMask_14}
    B {from=MaterialExpressionComponentMask_13}
    Alpha {from=MaterialExpressionScalarParameter_62}
   ]
  }
  MaterialExpressionLinearInterpolate_27 {
   class=MaterialExpressionLinearInterpolate
   inputs [
    A {from=MaterialExpressionComponentMask_11}
    B {from=MaterialExpressionComponentMask_12}
    Alpha {from=MaterialExpressionScalarParameter_65}
   ]
  }
  MaterialExpressionLinearInterpolate_3 {
   class=MaterialExpressionLinearInterpolate
   inputs [
    A {from=""}
    B {from=""}
    Alpha {from=MaterialExpressionDynamicParameter_3, output=Dissolve Outline}
   ]
   properties {constA=-1, constAlpha=1}
  }
  MaterialExpressionLinearInterpolate_4 {
   class=MaterialExpressionLinearInterpolate
   inputs [
    A {from=MaterialExpressionVectorParameter_4, output=RGB}
    B {from=MaterialExpressionLinearInterpolate_24}
    Alpha {from=MaterialExpressionDynamicParameter_2, output=Color Change}
   ]
  }
  MaterialExpressionLinearInterpolate_5 {
   class=MaterialExpressionLinearInterpolate
   inputs [
    A {from=MaterialExpressionReroute_0}
    B {from=MaterialExpressionMultiply_18}
    Alpha {from=MaterialExpressionScalarParameter_7}
   ]
  }
  MaterialExpressionLinearInterpolate_6 {
   class=MaterialExpressionLinearInterpolate
   inputs [
    A {from=MaterialExpressionAdd_0}
    B {from=MaterialExpressionAdd_5}
    Alpha {from=MaterialExpressionScalarParameter_7}
   ]
  }
  MaterialExpressionLinearInterpolate_7 {
   class=MaterialExpressionLinearInterpolate
   inputs [
    A {from=MaterialExpressionVectorParameter_7, output=RGB}
    B {from=MaterialExpressionVectorParameter_3, output=RGB}
    Alpha {from=MaterialExpressionSaturate_0}
   ]
  }
  MaterialExpressionLinearInterpolate_8 {
   class=MaterialExpressionLinearInterpolate
   inputs [
    A {from=MaterialExpressionShadingPathSwitch_0}
    B {from=MaterialExpressionNamedRerouteUsage_5}
    Alpha {from=MaterialExpressionScalarParameter_31}
   ]
  }
  MaterialExpressionLinearInterpolate_9 {
   class=MaterialExpressionLinearInterpolate
   inputs [
    A {from=MaterialExpressionScalarParameter_38}
    B {from=MaterialExpressionScalarParameter_39}
    Alpha {from=MaterialExpressionMultiply_27}
   ]
  }
  MaterialExpressionLocalPosition_0 {class=MaterialExpressionLocalPosition, inputs [], properties {}}
  MaterialExpressionMaterialFunctionCall_1 {
   class=MaterialExpressionMaterialFunctionCall
   inputs [Emissive {from=MaterialExpressionLinearInterpolate_6}]
   properties {materialFunction=&MF_FX_EmissiveAdjustments}
  }
  MaterialExpressionMaterialFunctionCall_2 {
   class=MaterialExpressionMaterialFunctionCall
   inputs [UV Channel {from=MaterialExpressionAdd_3}]
   properties {materialFunction=&LinearGradient}
  }
  MaterialExpressionMaterialFunctionCall_4 {
   class=MaterialExpressionMaterialFunctionCall
   inputs [UV Channel {from=MaterialExpressionAdd_7}]
   properties {materialFunction=&LinearGradient}
  }
  MaterialExpressionMaterialFunctionCall_5 {
   class=MaterialExpressionMaterialFunctionCall
   inputs [UV Channel {from=MaterialExpressionAdd_10}]
   properties {materialFunction=&LinearGradient}
  }
  MaterialExpressionMultiply_0 {
   class=MaterialExpressionMultiply
   inputs [
    A {from=MaterialExpressionAdd_6}
    B {from=MaterialExpressionLinearInterpolate_11}
   ]
  }
  MaterialExpressionMultiply_1 {
   class=MaterialExpressionMultiply
   inputs [
    A {from=MaterialExpressionPower_1}
    B {from=MaterialExpressionMultiply_13}
   ]
   properties {constA=0.2}
  }
  MaterialExpressionMultiply_10 {
   class=MaterialExpressionMultiply
   inputs [
    A {from=MaterialExpressionLinearInterpolate_15}
    B {from=MaterialExpressionAppendVector_2}
   ]
  }
  MaterialExpressionMultiply_11 {
   class=MaterialExpressionMultiply
   inputs [
    A {from=MaterialExpressionTextureSampleParameter2D_1, output=R}
    B {from=MaterialExpressionLinearInterpolate_20}
   ]
  }
  MaterialExpressionMultiply_12 {
   class=MaterialExpressionMultiply
   inputs [
    A {from=MaterialExpressionSubtract_3}
    B {from=MaterialExpressionScalarParameter_10}
   ]
   properties {constB=0.075}
  }
  MaterialExpressionMultiply_13 {
   class=MaterialExpressionMultiply
   inputs [
    A {from=MaterialExpressionMultiply_11}
    B {from=MaterialExpressionScalarParameter_3}
   ]
   properties {constB=0.8}
  }
  MaterialExpressionMultiply_14 {
   class=MaterialExpressionMultiply
   inputs [
    A {from=MaterialExpressionLinearInterpolate_19}
    B {from=MaterialExpressionVectorParameter_6, output=RGB}
   ]
  }
  MaterialExpressionMultiply_15 {
   class=MaterialExpressionMultiply
   inputs [
    A {from=MaterialExpressionPower_3}
    B {from=""}
   ]
   properties {constB=4}
  }
  MaterialExpressionMultiply_16 {
   class=MaterialExpressionMultiply
   inputs [
    A {from=MaterialExpressionNamedRerouteUsage_2}
    B {from=MaterialExpressionPower_5}
   ]
   properties {constA=0.2}
  }
  MaterialExpressionMultiply_17 {
   class=MaterialExpressionMultiply
   inputs [
    A {from=MaterialExpressionLinearInterpolate_3}
    B {from=MaterialExpressionMultiply_15}
   ]
  }
  MaterialExpressionMultiply_18 {
   class=MaterialExpressionMultiply
   inputs [
    A {from=MaterialExpressionParticleColor_1, output=A}
    B {from=MaterialExpressionReroute_0}
   ]
  }
  MaterialExpressionMultiply_19 {
   class=MaterialExpressionMultiply
   inputs [
    A {from=MaterialExpressionPower_2}
    B {from=MaterialExpressionMultiply_29}
   ]
   properties {constB=0.35}
  }
  MaterialExpressionMultiply_2 {
   class=MaterialExpressionMultiply
   inputs [
    A {from=MaterialExpressionFrac_0}
    B {from=MaterialExpressionOneMinus_1}
   ]
  }
  MaterialExpressionMultiply_20 {
   class=MaterialExpressionMultiply
   inputs [
    A {from=MaterialExpressionSmoothStep_7}
    B {from=MaterialExpressionScalarParameter_22}
   ]
  }
  MaterialExpressionMultiply_21 {
   class=MaterialExpressionMultiply
   inputs [
    A {from=MaterialExpressionLinearInterpolate_4}
    B {from=MaterialExpressionParticleColor_1, output=RGB}
   ]
  }
  MaterialExpressionMultiply_22 {
   class=MaterialExpressionMultiply
   inputs [
    A {from=MaterialExpressionPower_7}
    B {from=MaterialExpressionScalarParameter_32}
   ]
  }
  MaterialExpressionMultiply_23 {
   class=MaterialExpressionMultiply
   inputs [
    A {from=MaterialExpressionNamedRerouteUsage_6}
    B {from=MaterialExpressionLinearInterpolate_9}
   ]
  }
  MaterialExpressionMultiply_24 {
   class=MaterialExpressionMultiply
   inputs [
    A {from=MaterialExpressionTextureSampleParameter2D_5, output=RGB}
    B {from=MaterialExpressionScalarParameter_33}
   ]
  }
  MaterialExpressionMultiply_25 {
   class=MaterialExpressionMultiply
   inputs [
    A {from=MaterialExpressionDivide_0}
    B {from=MaterialExpressionDivide_1}
   ]
  }
  MaterialExpressionMultiply_26 {
   class=MaterialExpressionMultiply
   inputs [
    A {from=MaterialExpressionMultiply_25}
    B {from=MaterialExpressionDivide_2}
   ]
  }
  MaterialExpressionMultiply_27 {
   class=MaterialExpressionMultiply
   inputs [
    A {from=MaterialExpressionPower_8}
    B {from=MaterialExpressionScalarParameter_37}
   ]
  }
  MaterialExpressionMultiply_28 {
   class=MaterialExpressionMultiply
   inputs [
    A {from=MaterialExpressionMultiply_16}
    B {from=MaterialExpressionSmoothStep_3}
   ]
  }
  MaterialExpressionMultiply_29 {
   class=MaterialExpressionMultiply
   inputs [
    A {from=MaterialExpressionLinearInterpolate_1}
    B {from=MaterialExpressionScalarParameter_12}
   ]
   properties {constB=6}
  }
  MaterialExpressionMultiply_3 {
   class=MaterialExpressionMultiply
   inputs [
    A {from=MaterialExpressionMultiply_2}
    B {from=""}
   ]
   properties {constB=4}
  }
  MaterialExpressionMultiply_30 {
   class=MaterialExpressionMultiply
   inputs [
    A {from=MaterialExpressionMultiply_28}
    B {from=MaterialExpressionScalarParameter_20}
   ]
   properties {constB=25}
  }
  MaterialExpressionMultiply_31 {
   class=MaterialExpressionMultiply
   inputs [
    A {from=MaterialExpressionVertexNormalWS_0}
    B {from=MaterialExpressionMultiply_24}
   ]
  }
  MaterialExpressionMultiply_32 {
   class=MaterialExpressionMultiply
   inputs [
    A {from=MaterialExpressionPower_4}
    B {from=""}
   ]
   properties {constB=7}
  }
  MaterialExpressionMultiply_33 {
   class=MaterialExpressionMultiply
   inputs [
    A {from=MaterialExpressionMultiply_19}
    B {from=MaterialExpressionMultiply_32}
   ]
  }
  MaterialExpressionMultiply_34 {
   class=MaterialExpressionMultiply
   inputs [
    A {from=MaterialExpressionLinearInterpolate_14}
    B {from=MaterialExpressionDynamicParameter_4, output=Flicker}
   ]
  }
  MaterialExpressionMultiply_35 {
   class=MaterialExpressionMultiply
   inputs [
    A {from=MaterialExpressionNamedRerouteUsage_3}
    B {from=MaterialExpressionParticleColor_0, output=A}
   ]
  }
  MaterialExpressionMultiply_36 {
   class=MaterialExpressionMultiply
   inputs [
    A {from=MaterialExpressionComponentMask_2}
    B {from=MaterialExpressionAppendVector_4}
   ]
   properties {constB=0.01}
  }
  MaterialExpressionMultiply_37 {
   class=MaterialExpressionMultiply
   inputs [
    A {from=MaterialExpressionComponentMask_6}
    B {from=MaterialExpressionScalarParameter_44}
   ]
  }
  MaterialExpressionMultiply_38 {
   class=MaterialExpressionMultiply
   inputs [
    A {from=MaterialExpressionComponentMask_7}
    B {from=MaterialExpressionScalarParameter_45}
   ]
  }
  MaterialExpressionMultiply_39 {
   class=MaterialExpressionMultiply
   inputs [
    A {from=MaterialExpressionSubtract_1}
    B {from=MaterialExpressionScalarParameter_19}
   ]
   properties {constB=0.1}
  }
  MaterialExpressionMultiply_4 {
   class=MaterialExpressionMultiply
   inputs [
    A {from=MaterialExpressionPower_0}
    B {from=MaterialExpressionScalarParameter_5}
   ]
   properties {constB=5}
  }
  MaterialExpressionMultiply_40 {
   class=MaterialExpressionMultiply
   inputs [
    A {from=MaterialExpressionSmoothStep_2}
    B {from=MaterialExpressionScalarParameter_16}
   ]
   properties {constB=1.5}
  }
  MaterialExpressionMultiply_41 {
   class=MaterialExpressionMultiply
   inputs [
    A {from=MaterialExpressionDynamicParameter_5, output=Mask_R_Opacity}
    B {from=MaterialExpressionTextureSampleParameter2D_0, output=R}
   ]
  }
  MaterialExpressionMultiply_42 {
   class=MaterialExpressionMultiply
   inputs [
    A {from=MaterialExpressionTextureCoordinate_8}
    B {from=MaterialExpressionAppendVector_9}
   ]
  }
  MaterialExpressionMultiply_43 {
   class=MaterialExpressionMultiply
   inputs [
    A {from=MaterialExpressionComponentMask_9}
    B {from=MaterialExpressionDynamicParameter_6, output=Pattern UV Mult X}
   ]
  }
  MaterialExpressionMultiply_44 {
   class=MaterialExpressionMultiply
   inputs [
    A {from=MaterialExpressionComponentMask_10}
    B {from=MaterialExpressionDynamicParameter_6, output=Pattern UV Mutl Y}
   ]
  }
  MaterialExpressionMultiply_45 {
   class=MaterialExpressionMultiply
   inputs [
    A {from=MaterialExpressionAppendVector_12}
    B {from=MaterialExpressionTime_2}
   ]
  }
  MaterialExpressionMultiply_46 {
   class=MaterialExpressionMultiply
   inputs [
    A {from=MaterialExpressionTextureCoordinate_9}
    B {from=MaterialExpressionAppendVector_11}
   ]
  }
  MaterialExpressionMultiply_47 {
   class=MaterialExpressionMultiply
   inputs [
    A {from=MaterialExpressionAppendVector_13}
    B {from=MaterialExpressionTime_3}
   ]
  }
  MaterialExpressionMultiply_48 {
   class=MaterialExpressionMultiply
   inputs [
    A {from=MaterialExpressionTextureSampleParameter2D_6, output=RGB}
    B {from=MaterialExpressionScalarParameter_60}
   ]
  }
  MaterialExpressionMultiply_49 {
   class=MaterialExpressionMultiply
   inputs [
    A {from=MaterialExpressionNamedRerouteUsage_9}
    B {from=MaterialExpressionScalarParameter_61}
   ]
  }
  MaterialExpressionMultiply_5 {
   class=MaterialExpressionMultiply
   inputs [
    A {from=MaterialExpressionMultiply_26}
    B {from=MaterialExpressionScalarParameter_1}
   ]
  }
  MaterialExpressionMultiply_50 {
   class=MaterialExpressionMultiply
   inputs [
    A {from=MaterialExpressionRotator_2}
    B {from=MaterialExpressionConvert_1}
   ]
  }
  MaterialExpressionMultiply_51 {
   class=MaterialExpressionMultiply
   inputs [
    A {from=MaterialExpressionFrac_2}
    B {from=MaterialExpressionOneMinus_5}
   ]
  }
  MaterialExpressionMultiply_52 {
   class=MaterialExpressionMultiply
   inputs [
    A {from=MaterialExpressionMultiply_51}
    B {from=""}
   ]
   properties {constB=4}
  }
  MaterialExpressionMultiply_53 {
   class=MaterialExpressionMultiply
   inputs [
    A {from=MaterialExpressionPower_10}
    B {from=MaterialExpressionScalarParameter_55}
   ]
   properties {constB=5}
  }
  MaterialExpressionMultiply_54 {
   class=MaterialExpressionMultiply
   inputs [
    A {from=MaterialExpressionFrac_3}
    B {from=MaterialExpressionOneMinus_6}
   ]
  }
  MaterialExpressionMultiply_55 {
   class=MaterialExpressionMultiply
   inputs [
    A {from=MaterialExpressionMultiply_54}
    B {from=""}
   ]
   properties {constB=4}
  }
  MaterialExpressionMultiply_56 {
   class=MaterialExpressionMultiply
   inputs [
    A {from=MaterialExpressionPower_11}
    B {from=MaterialExpressionScalarParameter_57}
   ]
   properties {constB=5}
  }
  MaterialExpressionMultiply_6 {
   class=MaterialExpressionMultiply
   inputs [
    A {from=MaterialExpressionTextureCoordinate_4}
    B {from=MaterialExpressionAppendVector_0}
   ]
  }
  MaterialExpressionMultiply_7 {
   class=MaterialExpressionMultiply
   inputs [
    A {from=MaterialExpressionAppendVector_1}
    B {from=MaterialExpressionTime_1}
   ]
  }
  MaterialExpressionMultiply_8 {
   class=MaterialExpressionMultiply
   inputs [
    A {from=MaterialExpressionLinearInterpolate_19}
    B {from=MaterialExpressionMultiply_1}
   ]
  }
  MaterialExpressionMultiply_9 {
   class=MaterialExpressionMultiply
   inputs [
    A {from=MaterialExpressionMultiply_40}
    B {from=MaterialExpressionSmoothStep_0}
   ]
  }
  MaterialExpressionNamedRerouteDeclaration_0 {
   class=MaterialExpressionNamedRerouteDeclaration
   inputs [None {from=MaterialExpressionTextureSampleParameter2D_0, output=B}]
   properties {name=Mask Blue, nodeColor=<rgba 0.4062 0 1 1>}
  }
  MaterialExpressionNamedRerouteDeclaration_1 {
   class=MaterialExpressionNamedRerouteDeclaration
   inputs [None {from=MaterialExpressionLinearInterpolate_10}]
   properties {name=WPO, nodeColor=<rgba 0 1 0.9766 1>}
  }
  MaterialExpressionNamedRerouteDeclaration_2 {
   class=MaterialExpressionNamedRerouteDeclaration
   inputs [None {from=MaterialExpressionMultiply_9}]
   properties {name=DissolveLineHighlight, nodeColor=<rgba 1 0 0.5859 1>}
  }
  MaterialExpressionNamedRerouteDeclaration_3 {
   class=MaterialExpressionNamedRerouteDeclaration
   inputs [None {from=MaterialExpressionAdd_28}]
   properties {name=Dissolve, nodeColor=<rgba 1 0.9609 0 1>}
  }
  MaterialExpressionNamedRerouteDeclaration_4 {
   class=MaterialExpressionNamedRerouteDeclaration
   inputs [None {from=MaterialExpressionMultiply_13}]
   properties {name=PatternTex, nodeColor=<rgba 0.6406 0 1 1>}
  }
  MaterialExpressionNamedRerouteDeclaration_5 {
   class=MaterialExpressionNamedRerouteDeclaration
   inputs [None {from=MaterialExpressionMultiply_23}]
   properties {name=Vertex Color Opacity, nodeColor=<rgba 1 0.6563 0 1>}
  }
  MaterialExpressionNamedRerouteDeclaration_6 {
   class=MaterialExpressionNamedRerouteDeclaration
   inputs [None {from=MaterialExpressionShadingPathSwitch_0}]
   properties {name=To Vertex Color, nodeColor=<rgba 0.3359 0 1 1>}
  }
  MaterialExpressionNamedRerouteDeclaration_7 {
   class=MaterialExpressionNamedRerouteDeclaration
   inputs [None {from=MaterialExpressionTextureSampleParameter2D_0, output=G}]
   properties {name=Mask Green, nodeColor=<rgba 0.9688 1 0 1>}
  }
  MaterialExpressionNamedRerouteDeclaration_8 {
   class=MaterialExpressionNamedRerouteDeclaration
   inputs [None {from=MaterialExpressionSaturate_0}]
   properties {name=ColorRamp, nodeColor=<rgba 0.8281 0 1 1>}
  }
  MaterialExpressionNamedRerouteDeclaration_9 {
   class=MaterialExpressionNamedRerouteDeclaration
   inputs [None {from=MaterialExpressionMultiply_48}]
   properties {name=ColorTexture, nodeColor=<rgba 0.1953 1 0 1>}
  }
  MaterialExpressionNamedRerouteUsage_0 {class=MaterialExpressionNamedRerouteUsage}
  MaterialExpressionNamedRerouteUsage_1 {class=MaterialExpressionNamedRerouteUsage}
  MaterialExpressionNamedRerouteUsage_2 {class=MaterialExpressionNamedRerouteUsage}
  MaterialExpressionNamedRerouteUsage_3 {class=MaterialExpressionNamedRerouteUsage}
  MaterialExpressionNamedRerouteUsage_4 {class=MaterialExpressionNamedRerouteUsage}
  MaterialExpressionNamedRerouteUsage_5 {class=MaterialExpressionNamedRerouteUsage}
  MaterialExpressionNamedRerouteUsage_6 {class=MaterialExpressionNamedRerouteUsage}
  MaterialExpressionNamedRerouteUsage_7 {class=MaterialExpressionNamedRerouteUsage}
  MaterialExpressionNamedRerouteUsage_8 {class=MaterialExpressionNamedRerouteUsage}
  MaterialExpressionNamedRerouteUsage_9 {class=MaterialExpressionNamedRerouteUsage}
  MaterialExpressionOneMinus_0 {class=MaterialExpressionOneMinus, inputs [None {from=MaterialExpressionLinearInterpolate_27}]}
  MaterialExpressionOneMinus_1 {class=MaterialExpressionOneMinus, inputs [None {from=MaterialExpressionFrac_0}]}
  MaterialExpressionOneMinus_2 {class=MaterialExpressionOneMinus, inputs [None {from=MaterialExpressionLinearInterpolate_17}]}
  MaterialExpressionOneMinus_4 {class=MaterialExpressionOneMinus, inputs [None {from=MaterialExpressionAdd_13}]}
  MaterialExpressionOneMinus_5 {class=MaterialExpressionOneMinus, inputs [None {from=MaterialExpressionFrac_2}]}
  MaterialExpressionOneMinus_6 {class=MaterialExpressionOneMinus, inputs [None {from=MaterialExpressionFrac_3}]}
  MaterialExpressionPanner_0 {
   class=MaterialExpressionPanner
   inputs [
    Coordinate {from=MaterialExpressionAdd_17}
    Time {from=""}
    Speed {from=MaterialExpressionAppendVector_5}
   ]
   properties {speedY=-0.5}
  }
  MaterialExpressionPanner_1 {
   class=MaterialExpressionPanner
   inputs [
    Coordinate {from=MaterialExpressionMultiply_10}
    Time {from=""}
    Speed {from=MaterialExpressionAppendVector_3}
   ]
  }
  MaterialExpressionPanner_2 {
   class=MaterialExpressionPanner
   inputs [
    Coordinate {from=MaterialExpressionAppendVector_7}
    Time {from=""}
    Speed {from=MaterialExpressionAppendVector_6}
   ]
   properties {speedY=0.1}
  }
  MaterialExpressionPanner_3 {
   class=MaterialExpressionPanner
   inputs [
    Coordinate {from=MaterialExpressionMultiply_50}
    Time {from=""}
    Speed {from=MaterialExpressionConvert_2}
   ]
  }
  MaterialExpressionParticleColor_0 {class=MaterialExpressionParticleColor, inputs [], properties {}}
  MaterialExpressionParticleColor_1 {class=MaterialExpressionParticleColor, inputs [], properties {}}
  MaterialExpressionParticleColor_2 {class=MaterialExpressionParticleColor, inputs [], properties {}}
  MaterialExpressionPower_0 {
   class=MaterialExpressionPower
   inputs [
    Base {from=MaterialExpressionMultiply_3}
    Exp {from=MaterialExpressionScalarParameter_4}
   ]
   properties {constExponent=40}
  }
  MaterialExpressionPower_1 {
   class=MaterialExpressionPower
   inputs [
    Base {from=MaterialExpressionOneMinus_0}
    Exp {from=MaterialExpressionScalarParameter_2}
   ]
   properties {constExponent=1.5}
  }
  MaterialExpressionPower_10 {
   class=MaterialExpressionPower
   inputs [
    Base {from=MaterialExpressionMultiply_52}
    Exp {from=MaterialExpressionScalarParameter_54}
   ]
   properties {constExponent=40}
  }
  MaterialExpressionPower_11 {
   class=MaterialExpressionPower
   inputs [
    Base {from=MaterialExpressionMultiply_55}
    Exp {from=MaterialExpressionScalarParameter_56}
   ]
   properties {constExponent=40}
  }
  MaterialExpressionPower_2 {
   class=MaterialExpressionPower
   inputs [
    Base {from=MaterialExpressionTextureSampleParameter2D_0, output=G}
    Exp {from=MaterialExpressionScalarParameter_0}
   ]
  }
  MaterialExpressionPower_3 {
   class=MaterialExpressionPower
   inputs [
    Base {from=MaterialExpressionTextureSampleParameter2D_4, output=G}
    Exp {from=MaterialExpressionScalarParameter_21}
   ]
  }
  MaterialExpressionPower_4 {
   class=MaterialExpressionPower
   inputs [
    Base {from=MaterialExpressionComponentMask_3}
    Exp {from=""}
   ]
   properties {constExponent=5}
  }
  MaterialExpressionPower_5 {
   class=MaterialExpressionPower
   inputs [
    Base {from=MaterialExpressionComponentMask_4}
    Exp {from=MaterialExpressionScalarParameter_25}
   ]
   properties {constExponent=1.5}
  }
  MaterialExpressionPower_6 {
   class=MaterialExpressionPower
   inputs [
    Base {from=MaterialExpressionOneMinus_4}
    Exp {from=MaterialExpressionScalarParameter_6}
   ]
  }
  MaterialExpressionPower_7 {
   class=MaterialExpressionPower
   inputs [
    Base {from=MaterialExpressionLinearInterpolate_22}
    Exp {from=MaterialExpressionScalarParameter_30}
   ]
  }
  MaterialExpressionPower_8 {
   class=MaterialExpressionPower
   inputs [
    Base {from=MaterialExpressionVertexColor_0, output=R}
    Exp {from=MaterialExpressionScalarParameter_36}
   ]
  }
  MaterialExpressionReroute_0 {class=MaterialExpressionReroute, inputs [None {from=MaterialExpressionMultiply_0}]}
  MaterialExpressionRotator_0 {
   class=MaterialExpressionRotator
   inputs [
    Coordinate {from=MaterialExpressionTextureCoordinate_10, output=""}
    Time {from=MaterialExpressionScalarParameter_63, output=""}
   ]
   properties {}
  }
  MaterialExpressionRotator_1 {
   class=MaterialExpressionRotator
   inputs [
    Coordinate {from=MaterialExpressionTextureCoordinate_0, output=""}
    Time {from=MaterialExpressionScalarParameter_64, output=""}
   ]
   properties {}
  }
  MaterialExpressionRotator_2 {
   class=MaterialExpressionRotator
   inputs [
    Coordinate {from=MaterialExpressionTextureCoordinate_11, output=""}
    Time {from=MaterialExpressionScalarParameter_66, output=""}
   ]
   properties {}
  }
  MaterialExpressionSaturate_0 {class=MaterialExpressionSaturate, inputs [None {from=MaterialExpressionMultiply_22}]}
  MaterialExpressionSaturate_1 {class=MaterialExpressionSaturate, inputs [None {from=MaterialExpressionAdd_9}]}
  MaterialExpressionScalarParameter_0 {class=MaterialExpressionScalarParameter, properties {defaultValue=2, group=01 - Main, parameterName=Glow Power, sortPriority=4}}
  MaterialExpressionScalarParameter_1 {class=MaterialExpressionScalarParameter, properties {defaultValue=1, group=01 - Main, parameterName=Glow Flicker Amplitude, sortPriority=4}}
  MaterialExpressionScalarParameter_10 {
   class=MaterialExpressionScalarParameter
   properties {defaultValue=0.075, group=04 - Dissolve, parameterName=Main Dissolve Size, sortPriority=2}
  }
  MaterialExpressionScalarParameter_11 {
   class=MaterialExpressionScalarParameter
   properties {defaultValue=0.05, group=04 - Dissolve, parameterName=Dissolve Line Size, sortPriority=7}
  }
  MaterialExpressionScalarParameter_12 {class=MaterialExpressionScalarParameter, properties {defaultValue=6, group=01 - Main, parameterName=Glow Flicker Intensity, sortPriority=3}}
  MaterialExpressionScalarParameter_13 {class=MaterialExpressionScalarParameter, properties {defaultValue=0.25, group=01 - Main, parameterName=Glow Flciker Min}}
  MaterialExpressionScalarParameter_14 {class=MaterialExpressionScalarParameter, properties {defaultValue=0.75, group=01 - Main, parameterName=Glow Flciker Max}}
  MaterialExpressionScalarParameter_15 {class=MaterialExpressionScalarParameter, properties {group=01 - Main, parameterName=DepthFadeDistance, sortPriority=34}}
  MaterialExpressionScalarParameter_16 {
   class=MaterialExpressionScalarParameter
   properties {defaultValue=1.5, group=04 - Dissolve, parameterName=Dissolve Line Intensity, sortPriority=8}
  }
  MaterialExpressionScalarParameter_17 {
   class=MaterialExpressionScalarParameter
   properties {defaultValue=10, group=04 - Dissolve, parameterName=Dissolve Pattern Max, sortPriority=16}
  }
  MaterialExpressionScalarParameter_18 {class=MaterialExpressionScalarParameter, properties {group=04 - Dissolve, parameterName=Dissolve Pattern Min, sortPriority=15}}
  MaterialExpressionScalarParameter_19 {
   class=MaterialExpressionScalarParameter
   properties {defaultValue=0.1, group=04 - Dissolve, parameterName=Main Dissolve Distortion Strength, sortPriority=13}
  }
  MaterialExpressionScalarParameter_2 {
   class=MaterialExpressionScalarParameter
   properties {defaultValue=1.5, group=03 - Pattern, parameterName=Pattern Gradient Power, sortPriority=3}
  }
  MaterialExpressionScalarParameter_20 {
   class=MaterialExpressionScalarParameter
   properties {defaultValue=25, group=04 - Dissolve, parameterName=Dissolve Pattern Intensity, sortPriority=14}
  }
  MaterialExpressionScalarParameter_21 {
   class=MaterialExpressionScalarParameter
   properties {defaultValue=2, group=04 - Dissolve, parameterName=Dissolve Outline Power, sortPriority=18}
  }
  MaterialExpressionScalarParameter_22 {
   class=MaterialExpressionScalarParameter
   properties {defaultValue=1, group=04 - Dissolve, parameterName=Dissolve Outline Intensity, sortPriority=17}
  }
  MaterialExpressionScalarParameter_23 {
   class=MaterialExpressionScalarParameter
   properties {defaultValue=0.5, group=04 - Dissolve, parameterName=Dissolve Outline Max, sortPriority=20}
  }
  MaterialExpressionScalarParameter_24 {class=MaterialExpressionScalarParameter, properties {group=04 - Dissolve, parameterName=Dissolve Outline Min, sortPriority=19}}
  MaterialExpressionScalarParameter_25 {
   class=MaterialExpressionScalarParameter
   properties {defaultValue=1.5, group=03 - Pattern, parameterName=Pattern Gradient Power, sortPriority=3}
  }
  MaterialExpressionScalarParameter_26 {class=MaterialExpressionScalarParameter, properties {group=04 - Dissolve, parameterName=Main Dissolve Min, sortPriority=3}}
  MaterialExpressionScalarParameter_27 {
   class=MaterialExpressionScalarParameter
   properties {defaultValue=0.15, group=04 - Dissolve, parameterName=Main Dissolve Max, sortPriority=4}
  }
  MaterialExpressionScalarParameter_28 {class=MaterialExpressionScalarParameter, properties {group=04 - Dissolve, parameterName=Main Dissolve Min 2, sortPriority=5}}
  MaterialExpressionScalarParameter_29 {
   class=MaterialExpressionScalarParameter
   properties {defaultValue=0.2, group=04 - Dissolve, parameterName=Main Dissolve Max 2, sortPriority=6}
  }
  MaterialExpressionScalarParameter_3 {class=MaterialExpressionScalarParameter, properties {defaultValue=0.8, group=03 - Pattern, parameterName=Pattern Opacity, sortPriority=2}}
  MaterialExpressionScalarParameter_30 {class=MaterialExpressionScalarParameter, properties {defaultValue=1, group=02 - Color, parameterName=ColorRampPower, sortPriority=36}}
  MaterialExpressionScalarParameter_31 {class=MaterialExpressionScalarParameter, properties {group=06 - Vertex Color, parameterName=UseVertexColors, sortPriority=-1}}
  MaterialExpressionScalarParameter_32 {class=MaterialExpressionScalarParameter, properties {defaultValue=1, group=02 - Color, parameterName=ColorRampMult, sortPriority=36}}
  MaterialExpressionScalarParameter_33 {class=MaterialExpressionScalarParameter, properties {group=05 - WPO, parameterName=WPO Mult, sortPriority=4}}
  MaterialExpressionScalarParameter_34 {class=MaterialExpressionScalarParameter, properties {group=05 - WPO, parameterName=WPO Pan Speed X, sortPriority=2}}
  MaterialExpressionScalarParameter_35 {class=MaterialExpressionScalarParameter, properties {group=05 - WPO, parameterName=WPO Pan Speed Y, sortPriority=3}}
  MaterialExpressionScalarParameter_36 {
   class=MaterialExpressionScalarParameter
   properties {defaultValue=1, group=06 - Vertex Color, parameterName=VertexColor_Power, sortPriority=2}
  }
  MaterialExpressionScalarParameter_37 {
   class=MaterialExpressionScalarParameter
   properties {defaultValue=1, group=06 - Vertex Color, parameterName=VertexColor_Mult, sortPriority=3}
  }
  MaterialExpressionScalarParameter_38 {class=MaterialExpressionScalarParameter, properties {group=06 - Vertex Color, parameterName=VertexColor_Opacity_Min, sortPriority=0}}
  MaterialExpressionScalarParameter_39 {
   class=MaterialExpressionScalarParameter
   properties {defaultValue=1, group=06 - Vertex Color, parameterName=VertexColor_Opacity_Max, sortPriority=1}
  }
  MaterialExpressionScalarParameter_4 {class=MaterialExpressionScalarParameter, properties {defaultValue=40, group=03 - Pattern, parameterName=Highlight Power, sortPriority=5}}
  MaterialExpressionScalarParameter_40 {class=MaterialExpressionScalarParameter, properties {group=05 - WPO, parameterName=UseWPO, sortPriority=-2}}
  MaterialExpressionScalarParameter_41 {class=MaterialExpressionScalarParameter, properties {defaultValue=1, group=04 - Dissolve, parameterName=UseDissolve, sortPriority=-1}}
  MaterialExpressionScalarParameter_42 {class=MaterialExpressionScalarParameter, properties {group=08 - For Particles, parameterName=UseParticleCenterGlow, sortPriority=2}}
  MaterialExpressionScalarParameter_43 {class=MaterialExpressionScalarParameter, properties {group=04 - Dissolve, parameterName=UseFlicker, sortPriority=64}}
  MaterialExpressionScalarParameter_44 {class=MaterialExpressionScalarParameter, properties {defaultValue=1, group=05 - WPO, parameterName=WPO Texture U Tiling, sortPriority=0}}
  MaterialExpressionScalarParameter_45 {class=MaterialExpressionScalarParameter, properties {defaultValue=1, group=05 - WPO, parameterName=WPO Texture V Tiling, sortPriority=1}}
  MaterialExpressionScalarParameter_46 {class=MaterialExpressionScalarParameter, properties {group=08 - For Particles, parameterName=UseDynamic_R_Opacity, sortPriority=0}}
  MaterialExpressionScalarParameter_47 {class=MaterialExpressionScalarParameter, properties {group=06 - Vertex Color, parameterName=Use Added Outline}}
  MaterialExpressionScalarParameter_48 {class=MaterialExpressionScalarParameter, properties {group=03 - Pattern, parameterName=Use Dyn Parm UV}}
  MaterialExpressionScalarParameter_49 {class=MaterialExpressionScalarParameter, properties {group=02 - Color, parameterName=UseHorizontalGradient, sortPriority=37}}
  MaterialExpressionScalarParameter_5 {class=MaterialExpressionScalarParameter, properties {defaultValue=5, group=03 - Pattern, parameterName=Highlight Intensity, sortPriority=6}}
  MaterialExpressionScalarParameter_50 {class=MaterialExpressionScalarParameter, properties {group=02 - Color, parameterName=UseColorRampForPattern, sortPriority=38}}
  MaterialExpressionScalarParameter_51 {class=MaterialExpressionScalarParameter, properties {group=03 - Pattern, parameterName=UseAltHighlight}}
  MaterialExpressionScalarParameter_52 {class=MaterialExpressionScalarParameter, properties {group=02 - Color, parameterName=UseGreenVertColorForRamp, sortPriority=38}}
  MaterialExpressionScalarParameter_53 {class=MaterialExpressionScalarParameter, properties {group=02 - Color, parameterName=UseParticleAlphaOnly, sortPriority=33}}
  MaterialExpressionScalarParameter_54 {class=MaterialExpressionScalarParameter, properties {defaultValue=40, group=03 - Pattern, parameterName=Highlight Power, sortPriority=5}}
  MaterialExpressionScalarParameter_55 {class=MaterialExpressionScalarParameter, properties {defaultValue=5, group=03 - Pattern, parameterName=Highlight Intensity, sortPriority=6}}
  MaterialExpressionScalarParameter_56 {class=MaterialExpressionScalarParameter, properties {defaultValue=40, group=03 - Pattern, parameterName=Highlight Power, sortPriority=5}}
  MaterialExpressionScalarParameter_57 {class=MaterialExpressionScalarParameter, properties {defaultValue=5, group=03 - Pattern, parameterName=Highlight Intensity, sortPriority=6}}
  MaterialExpressionScalarParameter_58 {class=MaterialExpressionScalarParameter, properties {group=02 - Color, parameterName=UseTextureForColor, sortPriority=39}}
  MaterialExpressionScalarParameter_59 {class=MaterialExpressionScalarParameter, properties {group=02 - Color, parameterName=UseTextureForColor, sortPriority=39}}
  MaterialExpressionScalarParameter_6 {class=MaterialExpressionScalarParameter, properties {defaultValue=2, group=04 - Dissolve, parameterName=Dissolve Line Power, sortPriority=9}}
  MaterialExpressionScalarParameter_60 {class=MaterialExpressionScalarParameter, properties {defaultValue=1, group=02 - Color, parameterName=ColorTextureMult, sortPriority=39}}
  MaterialExpressionScalarParameter_61 {
   class=MaterialExpressionScalarParameter
   properties {defaultValue=1, group=02 - Color, parameterName=ColorTextureMult_Pattern, sortPriority=39}
  }
  MaterialExpressionScalarParameter_62 {class=MaterialExpressionScalarParameter, properties {defaultValue=1, group=04 - Dissolve, parameterName=UseUorVforDissolve?}}
  MaterialExpressionScalarParameter_63 {class=MaterialExpressionScalarParameter, properties {group=04 - Dissolve, parameterName=RotateDissolveDirection}}
  MaterialExpressionScalarParameter_64 {class=MaterialExpressionScalarParameter, properties {group=03 - Pattern, parameterName=PatternGradientRotation}}
  MaterialExpressionScalarParameter_65 {class=MaterialExpressionScalarParameter, properties {group=03 - Pattern, parameterName=UseUorVforGradient?}}
  MaterialExpressionScalarParameter_66 {class=MaterialExpressionScalarParameter, properties {group=02 - Color, parameterName=ColorTextureUVRotation}}
  MaterialExpressionScalarParameter_7 {class=MaterialExpressionScalarParameter, properties {group=02 - Color, parameterName=UseParticleColor, sliderMax=1}}
  MaterialExpressionScalarParameter_8 {
   class=MaterialExpressionScalarParameter
   properties {defaultValue=0.99, group=04 - Dissolve, parameterName=Dissolve Line Min, sortPriority=19}
  }
  MaterialExpressionScalarParameter_9 {
   class=MaterialExpressionScalarParameter
   properties {defaultValue=1.08, group=04 - Dissolve, parameterName=Dissolve Line Max, sortPriority=11}
  }
  MaterialExpressionShadingPathSwitch_0 {
   class=MaterialExpressionShadingPathSwitch
   inputs [
    Default {from=MaterialExpressionDepthFade_0, output=""}
    Deferred {from="", output=""}
    Forward {from="", output=""}
    Mobile {from=MaterialExpressionLinearInterpolate_23, output=""}
   ]
   properties {desc=Migrated from FeatureLevelSwitch}
  }
  MaterialExpressionSine_0 {class=MaterialExpressionSine, inputs [None {from=MaterialExpressionTime_0}], properties {period=0.1}}
  MaterialExpressionSine_1 {class=MaterialExpressionSine, inputs [None {from=MaterialExpressionTime_0}], properties {period=0.2}}
  MaterialExpressionSmoothStep_0 {
   class=MaterialExpressionSmoothStep
   inputs [
    Min {from=MaterialExpressionScalarParameter_28, output=""}
    Max {from=MaterialExpressionScalarParameter_29, output=""}
    Value {from=MaterialExpressionSmoothStep_1, output=""}
   ]
   properties {constMax=0.2}
  }
  MaterialExpressionSmoothStep_1 {
   class=MaterialExpressionSmoothStep
   inputs [
    Min {from=MaterialExpressionScalarParameter_26, output=""}
    Max {from=MaterialExpressionScalarParameter_27, output=""}
    Value {from=MaterialExpressionAdd_27, output=""}
   ]
   properties {constMax=0.15}
  }
  MaterialExpressionSmoothStep_2 {
   class=MaterialExpressionSmoothStep
   inputs [
    Min {from=MaterialExpressionScalarParameter_8, output=""}
    Max {from=MaterialExpressionScalarParameter_9, output=""}
    Value {from=MaterialExpressionPower_6, output=""}
   ]
   properties {constMax=1.08, constMin=0.99}
  }
  MaterialExpressionSmoothStep_3 {
   class=MaterialExpressionSmoothStep
   inputs [
    Min {from=MaterialExpressionScalarParameter_18, output=""}
    Max {from=MaterialExpressionScalarParameter_17, output=""}
    Value {from=MaterialExpressionLinearInterpolate_2, output=""}
   ]
   properties {constMax=10}
  }
  MaterialExpressionSmoothStep_7 {
   class=MaterialExpressionSmoothStep
   inputs [
    Min {from=MaterialExpressionScalarParameter_24, output=""}
    Max {from=MaterialExpressionScalarParameter_23, output=""}
    Value {from=MaterialExpressionMultiply_17, output=""}
   ]
   properties {constMax=0.5}
  }
  MaterialExpressionSubtract_1 {
   class=MaterialExpressionSubtract
   inputs [
    A {from=MaterialExpressionTextureSampleParameter2D_2, output=G}
    B {from=""}
   ]
   properties {constB=0.5}
  }
  MaterialExpressionSubtract_2 {
   class=MaterialExpressionSubtract
   inputs [
    A {from=MaterialExpressionLinearInterpolate_26}
    B {from=""}
   ]
  }
  MaterialExpressionSubtract_3 {
   class=MaterialExpressionSubtract
   inputs [
    A {from=MaterialExpressionTextureSampleParameter2D_3, output=G}
    B {from=""}
   ]
   properties {constB=0.5}
  }
  MaterialExpressionTextureCoordinate_0 {class=MaterialExpressionTextureCoordinate}
  MaterialExpressionTextureCoordinate_1 {class=MaterialExpressionTextureCoordinate}
  MaterialExpressionTextureCoordinate_10 {class=MaterialExpressionTextureCoordinate}
  MaterialExpressionTextureCoordinate_11 {class=MaterialExpressionTextureCoordinate}
  MaterialExpressionTextureCoordinate_2 {class=MaterialExpressionTextureCoordinate}
  MaterialExpressionTextureCoordinate_3 {class=MaterialExpressionTextureCoordinate}
  MaterialExpressionTextureCoordinate_4 {class=MaterialExpressionTextureCoordinate}
  MaterialExpressionTextureCoordinate_5 {class=MaterialExpressionTextureCoordinate}
  MaterialExpressionTextureCoordinate_6 {class=MaterialExpressionTextureCoordinate}
  MaterialExpressionTextureCoordinate_7 {class=MaterialExpressionTextureCoordinate}
  MaterialExpressionTextureCoordinate_8 {class=MaterialExpressionTextureCoordinate}
  MaterialExpressionTextureCoordinate_9 {class=MaterialExpressionTextureCoordinate}
  MaterialExpressionTextureSampleParameter2D_0 {
   class=MaterialExpressionTextureSampleParameter2D
   inputs [
    UVs {from="", output=""}
    Apply View MipBias {from="", output=""}
   ]
   properties {
    group=01 - Main
    parameterName=Main Tex
    sortPriority=0
    texture=&/T_FoundKelp_Glider_Blades
   }
  }
  MaterialExpressionTextureSampleParameter2D_1 {
   class=MaterialExpressionTextureSampleParameter2D
   inputs [
    UVs {from=MaterialExpressionPanner_1, output=""}
    Apply View MipBias {from="", output=""}
   ]
   properties {
    group=03 - Pattern
    parameterName=Pattern Tex
    sortPriority=0
    texture=&/T_FoundKelp_Pattern_Seamless
   }
  }
  MaterialExpressionTextureSampleParameter2D_2 {
   class=MaterialExpressionTextureSampleParameter2D
   inputs [
    UVs {from=MaterialExpressionVertexInterpolator_0, output=PS}
    Apply View MipBias {from="", output=""}
   ]
   properties {
    group=04 - Dissolve
    parameterName=Main Dissolve Distortion Tex
    sortPriority=12
    texture=&AtmosphericCloudNoise01
   }
  }
  MaterialExpressionTextureSampleParameter2D_3 {
   class=MaterialExpressionTextureSampleParameter2D
   inputs [
    UVs {from=MaterialExpressionPanner_0, output=""}
    Apply View MipBias {from="", output=""}
   ]
   properties {
    group=04 - Dissolve
    parameterName=Main Dissolve Tex
    samplerType=SAMPLERTYPE_Masks
    sortPriority=0
    texture=/BRCosmetics/Weapons/FORT_Melee/Pickaxe_ClaimReflect/Materials/T_Pickaxe_ClaimReflect_Mask
   }
  }
  MaterialExpressionTextureSampleParameter2D_4 {
   class=MaterialExpressionTextureSampleParameter2D
   inputs [
    UVs {from="", output=""}
    Apply View MipBias {from="", output=""}
   ]
   properties {
    group=01 - Main
    parameterName=Main Tex
    sortPriority=0
    texture=&T_FoundKelp_Glider_Blades
   }
  }
  MaterialExpressionTextureSampleParameter2D_5 {
   class=MaterialExpressionTextureSampleParameter2D
   inputs [
    UVs {from=MaterialExpressionPanner_2, output=""}
    Apply View MipBias {from="", output=""}
   ]
   properties {
    group=05 - WPO
    parameterName=WPO Texture
    sortPriority=-1
    texture=/Game/Characters/Player/Male/Medium/Bodies/M_MED_Viper/Textures/Glitch/T_SoftNoise
   }
  }
  MaterialExpressionTextureSampleParameter2D_6 {
   class=MaterialExpressionTextureSampleParameter2D
   inputs [
    UVs {from=MaterialExpressionPanner_3, output=""}
    Apply View MipBias {from="", output=""}
   ]
   properties {
    group=02 - Color
    parameterName=ColorTexture
    sortPriority=39
    texture=/BRCosmetics/Weapons/FORT_Melee/Pickaxe_Season_40/FX/T_FX_Season40_Pickaxe_Reload_C
   }
  }
  MaterialExpressionTime_0 {class=MaterialExpressionTime}
  MaterialExpressionTime_1 {class=MaterialExpressionTime}
  MaterialExpressionTime_2 {class=MaterialExpressionTime}
  MaterialExpressionTime_3 {class=MaterialExpressionTime}
  MaterialExpressionVectorParameter_0 {
   class=MaterialExpressionVectorParameter
   properties {group=03 - Pattern, parameterName=Pattern UV/Speed, sortPriority=1, defaultValue=<rgba 1 1 0 0>}
  }
  MaterialExpressionVectorParameter_1 {
   class=MaterialExpressionVectorParameter
   properties {group=03 - Pattern, parameterName=Highlight UV/Speed, sortPriority=4, defaultValue=<rgba 1 1 0 0.5>}
  }
  MaterialExpressionVectorParameter_2 {
   class=MaterialExpressionVectorParameter
   properties {group=04 - Dissolve, parameterName=Main Dissolve UV/Speed, sortPriority=1, defaultValue=<rgba 0.01 0.01 0 -0.5>}
  }
  MaterialExpressionVectorParameter_3 {
   class=MaterialExpressionVectorParameter
   properties {group=02 - Color, parameterName=Main Color, sortPriority=2, defaultValue=<rgba 1 0 0 0>}
  }
  MaterialExpressionVectorParameter_4 {
   class=MaterialExpressionVectorParameter
   properties {group=02 - Color, parameterName=Intro Color, sortPriority=1, defaultValue=<rgba 1 1 1 1>}
  }
  MaterialExpressionVectorParameter_5 {
   class=MaterialExpressionVectorParameter
   properties {group=02 - Color, parameterName=PatternColor, sortPriority=34, defaultValue=<rgba 1 1 1 1>}
  }
  MaterialExpressionVectorParameter_6 {
   class=MaterialExpressionVectorParameter
   properties {group=02 - Color, parameterName=HighlightColor, sortPriority=35, defaultValue=<rgba 1 1 1 1>}
  }
  MaterialExpressionVectorParameter_7 {
   class=MaterialExpressionVectorParameter
   properties {group=02 - Color, parameterName=Main Color Bottom, sortPriority=2, defaultValue=<rgba 1 0 0 0>}
  }
  MaterialExpressionVectorParameter_8 {
   class=MaterialExpressionVectorParameter
   properties {group=02 - Color, parameterName=PatternColor02, sortPriority=34, defaultValue=<rgba 1 1 1 1>}
  }
  MaterialExpressionVectorParameter_9 {class=MaterialExpressionVectorParameter, properties {group=02 - Color, parameterName=ColorTextureUVTimePan, defaultValue=<rgba 1 1 0 0>}}
  MaterialExpressionVertexColor_0 {class=MaterialExpressionVertexColor, inputs [], properties {}}
  MaterialExpressionVertexColor_1 {class=MaterialExpressionVertexColor, inputs [], properties {}}
  MaterialExpressionVertexInterpolator_0 {class=MaterialExpressionVertexInterpolator, inputs [VS {from=MaterialExpressionMultiply_36, output=""}], properties {}}
  MaterialExpressionVertexNormalWS_0 {class=MaterialExpressionVertexNormalWS, inputs [], properties {}}
 }
 properties {
  allowTranslucentCustomDepthWrites=true
  bAutomaticallySetUsageInEditor=false
  bUsedWithNiagaraMeshParticles=true
  bUsedWithNiagaraSprites=true
  blendMode=BLEND_Translucent
  shadingModel=MSM_Unlit
  twoSided=true
  usedShadingModels=Unlit
 }
}
