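# Niagara effect for the FoundKelp glider, layering blade sprites, flickering blade glows, a central glow, speedlines and ribbon trails.
# User.ForwardDot and RightDot steer speedlines, while ForwardDot also drives blade material parameters.
# The four-point trail bends procedurally by slerping an interpolated direction and adding a smoothed RightDot sideways offset.
# VariantSwitch picks glow and trail colors, with front-end preview and time-of-day modifiers adjusting intensity and layout.
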NiagaraSystem /BRCosmetics/Gadgets/Assets/VinderTech_GliderChute/Glider_FoundKelp/FX/NS_Glider_FoundKelp
dependencies {
 usedDataInterfaces [
  &NiagaraDataInterfaceCurve
  &NiagaraDataInterfaceVector2DCurve
  &NiagaraDataInterfaceColorCurve
  &NiagaraDataInterfaceArrayFloat
  &NiagaraDataInterfaceArrayFloat3
  &NiagaraDataInterfaceArrayFloat2
  &NiagaraDataInterfaceSkeletalMesh
 ]
 usedDynamicInputs [&Boolean_AndOperation]
 usedModules [
  &self:SystemSpawnScript.SystemScriptSource.SystemScriptGraph.NiagaraNodeAssignment_1.SetVariables_#1
  &SystemState
  &FN_Integrator
  &self:SystemSpawnScript.SystemScriptSource.SystemScriptGraph.NiagaraNodeAssignment_2.SetVariables_#2
 ]
}
emitters {
 BladeGlows {
  emitterUpdateScript [
   &EmitterState(
    Inactive Response=::NewEnumerator1
    Life Cycle Mode=::NewEnumerator1
    Loop Behavior=::NewEnumerator1
    Loop Duration Mode=::NewEnumerator1
    UseLoopDelay=true
   )
   &SpawnBurst_Instantaneous(Spawn Count=4, Use Spawn Probability=false, UseLoopCountLimit=false)
  ]
  particleSpawnScript [
   &FN_SetSpawnGroup()
   &InitializeParticle(
    Color=&MergeHelper_ColorSwitch(
     0=<rgba 1 0 0 1>
     1=<rgba 0.1092 0.4271 0.3944 1>
     2=<rgba 0.1851 0.0929 0.9387 1>
     3=?
     Index=&MakeIntFromFloat(Float=$User.VariantSwitch, Conversion Mode=::NewEnumerator1)
     Num=::NewEnumerator1
    )
    Lifetime=2
    Position Offset=&SelectVectorFromArray(
     Direct Array Index=&SpawnGroup()
     Vector Selection Array=DataInterface() {propertyValues {FloatData [<-122 15 -6>, <-65 -100 -6>, <122 15 -6>, <65 -100 -6>]}}
    )
    Sprite Size=&Multiply_Vector2D_ByFloat(
     Float=$System.GlobalScale
     Vector2D=&SelectVector2DFromArray(
      Direct Array Index=&SpawnGroup()
      Direct Array Mode=::NewEnumerator1
      Vector 2D Selection Array=DataInterface() {propertyValues {FloatData [<45 7.5>, <450 7.5>]}}
     )
    )
    Sprite Size Mode=::NewEnumerator4
    UsePositionOffset=true
    Write Material Random=true
   )
  ]
  particleUpdateScript [
   &ParticleState(Kill Particles When Lifetime Has Elapsed=false, Let Infinitely Lived Particles Die When Emitter Deactivates=true)
   &SpriteFacingAndAlignment(
    Sprite Alignment=<0 0 1>
    Sprite Facing=&SelectVectorFromArray(
     Direct Array Index=&SpawnGroup()
     Vector Selection Array=DataInterface() {propertyValues {FloatData [<0 1 0>, <1.3 0.5 0>, <0 1 0>, <1.3 -0.5 0>]}}
    )
    WriteAlignment=true
    WriteFacing=true
   )
   &DynamicMaterialParameters(Module.Param0WriteEnabled=true, Module.Param1WriteEnabled=true, Module.Param2WriteEnabled=true, Module.Param3WriteEnabled=true)
   &ScaleColor(
    Scale Alpha=&Multiply_Float(
     A=&FloatFromCurve(
      FloatCurve=DataInterface() {propertyValues {Curve {defaultValue=1, keys [<at 0.75 0>, {tangentMode=RCTM_User, time=1, value=1}]}}}
      CurveIndex=$Particles.Age
      Scale Curve=&Waveform(
       Global Clamp Min=0.25
       Global Clamp Max=0.4
       "Waveform [1]"=::NewEnumerator2
       "[1] Frequency"=2550
       "[1] Compound Frequency"=2000
       "[1] Amplitude Min/Max"=<0 1>
       "Waveform [2] Blend"=::NewEnumerator2
      )
     )
     B=&FN_FrontEndFloatModifier(Gameplay=1, Front End=0.5)
    )
    ScaleA=true
    ScaleRGB=true
   ) {module=ScaleColor001}
  ]
  rendererClasses [&NiagaraSpriteRendererProperties]
  renderers [{rendererClass=&NiagaraSpriteRendererProperties}]
 }
 Blades {
  emitterSpawnScript [/SetVariables() {enabled=false, isSetParameters=true}]
  emitterUpdateScript [
   &EmitterState(Life Cycle Mode=::NewEnumerator1, Loop Behavior=::NewEnumerator1, Loop Duration Mode=::NewEnumerator1, UseLoopDelay=true)
   &SpawnBurst_Instantaneous(Spawn Count=3, Use Spawn Probability=false, UseLoopCountLimit=false)
  ]
  particleSpawnScript [
   &FN_SetSpawnGroup(Number of Spawn Groups=3)
   &InitializeParticle(
    Lifetime=0.5
    Position Offset=&SelectVectorFromArray(
     Direct Array Index=&SpawnGroup()
     Vector Selection Array=DataInterface() {propertyValues {FloatData [<0 54 -6>, <0 54 -6.5>, <0 54 -5.5>]}}
    )
    Sprite Rotation Angle=180
    Sprite Rotation Mode=::NewEnumerator0
    Sprite Size Mode=::NewEnumerator3
    Uniform Sprite Size=&Multiply_Float(A=340, B=$System.GlobalScale)
    UsePositionOffset=true
    Write Material Random=true
   )
   &SpriteFacingAndAlignment(Sprite Facing=<0 0 1>, WriteAlignment=true, WriteFacing=true)
  ]
  particleUpdateScript [
   &ParticleState(Kill Particles When Lifetime Has Elapsed=false, Let Infinitely Lived Particles Die When Emitter Deactivates=true)
   &DynamicMaterialParameters(
    Index 0 Param 1=&FloatFromCurve(FloatCurve=DataInterface() {propertyValues {Curve {keys [<at 0.55 -0.2>,<at 1 1>]}}})
    Index 0 Param 2=&Lerp_Float(A=0.5, B=1.5, Alpha=$User.ForwardDot)
    Index 0 Param 3=&FloatFromCurve(FloatCurve=DataInterface() {propertyValues {Curve {keys [<at 0.15 0>,<at 0.4 1>]}}})
    Index 0 Param 4=&FloatFromCurve(FloatCurve=DataInterface() {propertyValues {Curve {keys [<at 0 0>,<at 0.25 1>]}}})
    Index 1 Param 1=&FloatFromCurve(FloatCurve=DataInterface() {propertyValues {Curve {keys [<at 0.4 0>,<at 0.8 1>]}}})
    Module.Index0Float0Write=true
    Module.Index0Float1Write=true
    Module.Index0Float2Write=true
    Module.Index0Float3Write=true
    Module.Index1Float0Write=true
    Module.Index1Float1Write=true
    Module.Index1Float2Write=true
    Module.Index1Float3Write=true
    Write Parameter Index 1=::NewEnumerator1
   )
  ]
  rendererClasses [&NiagaraSpriteRendererProperties]
  renderers [{rendererClass=&NiagaraSpriteRendererProperties}]
 }
 Glow {
  emitterUpdateScript [
   &EmitterState(Life Cycle Mode=::NewEnumerator1, Loop Behavior=::NewEnumerator1, Loop Duration Mode=::NewEnumerator1, UseLoopDelay=false)
   &SpawnBurst_Instantaneous(Use Spawn Probability=false, UseLoopCountLimit=false)
  ]
  particleSpawnScript [
   &InitializeParticle(
    Color=&MergeHelper_ColorSwitch(
     0=<rgba 1 0 0 1>
     1=<rgba 0.0625 0.8226 1 1>
     2=<rgba 0.9918 0.2292 1 1>
     3=?
     Index=&MakeIntFromFloat(Float=$User.VariantSwitch, Conversion Mode=::NewEnumerator1)
     Num=::NewEnumerator1
    )
    Lifetime=0.5
    Position Offset=<0 0 -7.5>
    Sprite Rotation Mode=::NewEnumerator0
    Sprite Size=&Multiply_Vector2D_ByFloat(Float=$System.GlobalScale, Vector2D=<100 55>)
    Sprite Size Mode=::NewEnumerator4
    UsePositionOffset=true
    Write Material Random=true
   )
   &CameraOffset(Camera Offset Amount=10)
  ]
  particleUpdateScript [
   &ParticleState(Kill Particles When Lifetime Has Elapsed=false, Let Infinitely Lived Particles Die When Emitter Deactivates=true)
   &ScaleColor(
    Scale Alpha=&FloatFromCurve(
     FloatCurve=DataInterface() {propertyValues {Curve {keys [
         {arriveTangentWeight=0.1667, interpMode=RCIM_Cubic, leaveTangentWeight=0.1667, tangentMode=RCTM_User, time=1.4, value=0}
         {tangentMode=RCTM_User, time=2, value=1}
        ]}}}
     CurveIndex=$Emitter.Age
     Scale Curve=&Multiply_Float(
      A=&SelectFloatFromArray(Direct Array Index=&SpawnGroup(), Float Selection Array=DataInterface() {propertyValues {FloatData [0.7,0.5]}})
      B=&Multiply_Float(
       A=&Lerp_Float(A=0.5, B=1, Alpha=$System.Timeline.Time)
       B=&FN_FrontEndFloatModifier(
        Gameplay=&FNTimeOfDayScalar(Night Scale=0.65, Morning Scale=0.8, Evening Scale=0.8)
        Front End=0.4
       )
      )
     )
    )
    ScaleA=true
    ScaleRGB=false
   )
  ]
  rendererClasses [&NiagaraSpriteRendererProperties]
  renderers [{rendererClass=&NiagaraSpriteRendererProperties}]
 }
 Speedlines {
  emitterUpdateScript [
   &EmitterState()
   &SpawnRate(
    SpawnRate=&MakeCustomFloatFromBool(Bool=$User.bIsFrontEndPreview, True Float=50, False Float=90)
    Use Spawn Probability=true
   ) {module=SpawnRate001}
  ]
  particleSpawnScript [
   &InitializeParticle(
    Color=&MakeLinearColorFromVectorAndFloat(
     "Vector (RGB)"=<1 1 1>
     "Float (Alpha)"=&Multiply_Float(
      A=&UniformRangedFloat(Minimum=1, Maximum=1)
      B=&MakeCustomFloatFromBool(
       Bool=$User.bIsFrontEndPreview
       True Float=&FloatFromCurve(
        FloatCurve=DataInterface() {propertyValues {Curve {keys [{tangentMode=RCTM_User, time=0, value=0}, {tangentMode=RCTM_User, time=2, value=1}]}}}
        CurveIndex=$Emitter.Age
        Scale Curve=0.4
       )
       False Float=1
      )
     )
    )
    Lifetime Max=0.2
    Lifetime Min=0.15
    Lifetime Mode=::NewEnumerator1
    Position Offset=&FN_FrontEndVector3Modifier(Gameplay=<0 -50 0>, Front End=<0 50 0>)
    Sprite Size Max=<3 10>
    Sprite Size Min=<2 2>
    Sprite Size Mode=::NewEnumerator2
    UsePositionOffset=true
    Write Material Random=true
   )
   &AddVelocity(
    Module.Use Falloff Distance=false
    Origin Offset=&FN_FrontEndVector3Modifier(Gameplay=<0 0 0>, Front End=<0 200 0>)
    Override Seed=false
    Velocity Mode=::NewEnumerator1
    Velocity Speed=&FN_FrontEndFloatModifier(Gameplay=-2500, Front End=-1400)
   ) {module=AddVelocity001}
   &TorusLocation(
    Align Torus With Axis=true
    Coordinate Space:ENiagaraCoordinateSpace=::Simulation
    Handle Radius=0
    Large Radius:NiagaraFloat=&FN_FrontEndFloatModifier(
     Gameplay=&RandomRangeFloat(Minimum=450, Maximum=500)
     Front End=&RandomRangeFloat(Minimum=350, Maximum=400)
    )
    MaskSpawnGroup=false
    Module.Override Randomness=false
    Offset=<0 33 0>
    Torus Axis=<0 1 0>
    Torus Origin:NiagaraPosition=&AddVectorToPosition(
     Position=$Particles.Position
     Vector=&FN_FrontEndVector3Modifier(Gameplay=<0 -500 0>, Front End=<0 -500 0>)
    )
   )
   &AddVelocity(
    Velocity=&Add_Vector(
     A=&RandomRangeVector(Minimum=<0 300 0>, Maximum=<0 700 1>)
     B=&Add_Vector(
      A=&Multiply_VectorByFloat(Vector=<-555 0 0>, Float=$User.RightDot)
      B=&Multiply_VectorByFloat(Vector=<0 0 222>, Float=$User.ForwardDot)
     )
    )
    Velocity Speed Scale=&FN_FrontEndFloatModifier(Gameplay=1, Front End=-0.3)
   )
  ]
  particleUpdateScript [
   &SolveForcesAndVelocity(Clamp Velocity=false, Limit Acceleration=false)
   &ParticleState()
   &ScaleSpriteSize(Scale Factor=&Vector2DFromCurve(
     Vector2Curve=DataInterface() {propertyValues {XCurve {keys [<at 0 0.5 RCIM_Cubic>,<at 0.5 1 RCIM_Cubic>]}, YCurve {keys [<at 0 0 RCIM_Cubic>,<at 0.5 1 RCIM_Cubic>]}}}
     ScaleCurve=&Vector2DFromFloat(Value=1)
    ))
   &ScaleSpriteSizeBySpeed(Max Scale Factor=<1 50>, Min Scale Factor=<0.01 0.05>, Velocity Threshold=1)
   &ScaleColor(
    Scale Alpha=&FloatFromCurve(
     FloatCurve=DataInterface() {propertyValues {Curve {keys [
         <at 0 0 RCIM_Cubic>
         <at 0.3759 1.0207>
         {arriveTangent=-2.221, interpMode=RCIM_Cubic, leaveTangent=-2.221, tangentMode=RCTM_User, time=1, value=0}
        ]}}}
     Scale Curve=&MakeCustomFloatFromBool(
      Bool=$User.bIsFrontEndPreview
      True Float=&MakeCustomFloatFromBool(Bool=$User.bIsFullyDeployed, True Float=0.3, False Float=0)
      False Float=&Lerp_Float(A=0, B=0.3, Alpha=&FN_FrontEndFloatModifier(Gameplay=$System.Timeline.Time, Front End=0.5))
     )
    )
    Scale RGB=&VectorFromFloat(Value=&FN_FrontEndFloatModifier(Gameplay=0.5, Front End=0.33))
    ScaleA=true
    ScaleRGB=true
   )
  ]
  rendererClasses [&NiagaraSpriteRendererProperties]
  renderers [{rendererClass=&NiagaraSpriteRendererProperties}]
 }
 Trail002 {
  emitterUpdateScript [
   &EmitterState(
    Inactive Response=::NewEnumerator1
    Life Cycle Mode=::NewEnumerator1
    Loop Behavior=::NewEnumerator1
    Loop Duration Mode=::NewEnumerator1
    UseLoopDelay=false
   )
   &Timeline(Play=&SetBoolByFloatComparison(A=$User.ForwardDot, B=0.1))
   &InterpolateOverTime(Target Float Value:NiagaraFloat=$User.RightDot)
   /SetVariables(
    Emitter.PositionOffset:Vector3f=&TransformVector(
     Vector=&FN_FrontEndVector3Modifier(Gameplay=<0 0 -2.5>, Front End=<0 0 -4.5>)
     Destination Space=::Simulation
    )
    Emitter.RibbonColor:LinearColor=&MergeHelper_ColorSwitch(
     0=&LerpLinearColors(
      StartColor=<rgba 1.7 0 0.0113 0.7>
      EndColor=<rgba 2 0 0 0.7>
      LerpFactor=&FN_FrontEndFloatModifier(Gameplay=0, Front End=1)
     )
     1=&LerpLinearColors(
      StartColor=<rgba 0.1107 0.3746 0.6719 0.8087>
      EndColor=<rgba 0.3634 0.5361 1 0.809>
      LerpFactor=&FN_FrontEndFloatModifier(Gameplay=0, Front End=1)
     )
     2=&LerpLinearColors(
      StartColor=<rgba 0.9317 0 1 0.7>
      EndColor=<rgba 0.15 0.063 0.7529 1>
      LerpFactor=&FN_FrontEndFloatModifier(Gameplay=&Lerp_Float(A=0, B=1, Alpha=0.92), Front End=1)
     )
     3=&LerpLinearColors(
      StartColor=<rgba 0.713 0.4617 0.2816 1>
      EndColor=<rgba 0.1016 0.0642 0.0373 1>
      LerpFactor=&FN_FrontEndFloatModifier(Gameplay=&Lerp_Float(A=0, B=1, Alpha=0.92), Front End=1)
     )
     Index=&MakeIntFromFloat(Float=$User.VariantSwitch, Conversion Mode=::NewEnumerator1)
     Num=::NewEnumerator1
    )
    Emitter.RibbonLength:NiagaraFloat=&FN_FrontEndFloatModifier(Gameplay=&Lerp_Float(A=50, B=125, Alpha=$StackContext.Timeline.Time), Front End=85)
   ) {isSetParameters=true}
   &SpawnBurst_Instantaneous(Spawn Count=4, Use Spawn Probability=false, UseLoopCountLimit=false)
   /BRCosmetics/Animation/Game/MainPlayer/Emotes/UpbeatIguana/FX/InterpolateTrailVector(
    Range=0.3
    Target Vector:Vector3f=&AddVectorToPosition(
     Position=&Multiply_VectorByFloat(Vector=$Engine.Owner.SystemYAxis, Float=&FN_FrontEndFloatModifier(Gameplay=1, Front End=1))
     Vector=&FN_FrontEndVector3Modifier(Gameplay=<0 0 -0.35>, Front End=<0 0 0>)
    )
   )
  ]
  particleSpawnScript [
   /SetVariables(Particles.RibbonWidth=&Multiply_Float(
     A=&Multiply_Float(A=29, B=&FN_FrontEndFloatModifier(Gameplay=1, Front End=0.62))
     B=&Lerp_Float(A=0.66, B=1, Alpha=&ReturnNormalizedExecIndex())
    )) {isSetParameters=true}
   &CameraOffset(Camera Offset Amount=50)
  ]
  particleUpdateScript [
   &SkeletalMeshLocation(Bone Sampling Mode=::NewEnumerator4, MaskSpawnGroup=false, Skeletal Mesh=$System.Skeletal Mesh)
   /SetVariables(Particles.Color=&MultiplyLinearColor(
     Linear Color=$Emitter.RibbonColor
     Scale Factor=&MakeVector4_FromFloat(Float=&Multiply_Float(
       A=$System.Timeline.Time
       B=&FNTimeOfDayScalar(Night Scale=0.5, Morning Scale=0.8, Evening Scale=0.8)
      ))
    )) {isSetParameters=true}
   /SetVariables(
    Particles.InterpolatedTrailVector:Vector3f=&TransformVector(
     Vector=/Niagara/DynamicInputs/Math/SlerpVector(
      Vector A:Vector3f=$Emitter.InterpolatedVector
      Vector B:Vector3f=$Emitter.TargetVector
      Alpha:NiagaraFloat=&ReturnNormalizedExecIndex()
      Scale:NiagaraFloat=1
     )
     Source Space=::World
     Destination Space=::Simulation
     Apply Scale=true
    )
    Particles.RibbonFacing=&TransformVector(Vector=$Engine.Owner.SystemZAxis, Source Space=::World, Destination Space=::Simulation)
   ) {isSetParameters=true}
   /SetVariables(Particles.SidewaysOffsetVector:Vector3f=&Multiply_VectorByFloat(
     Vector=&NormalizeVector(
      Vector To Normalize=/Niagara/DynamicInputs/Vectors/CrossProduct(
       Vector A:Vector3f=$StackContext.InterpolatedTrailVector
       Vector B:Vector3f=$Particles.RibbonFacing
       Normalize Incoming Vectors:NiagaraBool=true
      )
      Safe Normalize=true
      Safe Fallback Threshold=0
      Safe Fallback Vector=<0 0 0.001>
     )
     Float=&Multiply_Float(
      A=$Emitter.InterpolateOverTime.InterpolateOverTime_Float.Moving Average
      B=&Multiply_Float(
       A=&Power(a=&OneMinusFloat(Float=&ReturnNormalizedExecIndex()), b=2)
       B=-99
      )
     )
    )) {isSetParameters=true}
   /SetVariables(Particles.InterpolatedPosition:NiagaraPosition=&AddVectorToPosition(
     Position=&AddVectorToPosition(
      Position=$Engine.Emitter.SimulationPosition
      Vector=&Add_Vector(A=$Emitter.PositionOffset, B=$StackContext.SidewaysOffsetVector)
     )
     Vector=&Multiply_VectorByFloat(
      Vector=$Particles.InterpolatedTrailVector
      Float=&Multiply_Float(A=&OneMinusFloat(Float=&ReturnNormalizedExecIndex()), B=$Emitter.RibbonLength)
     )
    )) {isSetParameters=true}
   /SetVariables(Particles.Linear Color=&MergeHelper_ColorSwitch(
     0=&Multiply_LinearColorByFloat(
      Linear Color=<rgba 1 0 0 0>
      Float=&MakeFloatFromVector4(
       Vector 4=&MakeVector4_FromFloat(Float=&Multiply_Float(
         A=$System.Timeline.Time
         B=&FNTimeOfDayScalar(Night Scale=0.5, Morning Scale=0.8, Evening Scale=0.8)
        ))
       Channel=::NewEnumerator0
      )
     )
     1=&Multiply_LinearColorByFloat(
      Linear Color=&LerpLinearColors(
       StartColor=<rgba 0.1032 0.6931 0.901 0.8087>
       EndColor=<rgba 0.3634 0.5361 1 0.08>
       LerpFactor=&FN_FrontEndFloatModifier(Gameplay=0, Front End=0.37)
      )
      Float=&MakeFloatFromVector4(
       Vector 4=&MakeVector4_FromFloat(Float=&Multiply_Float(
         A=$System.Timeline.Time
         B=&FNTimeOfDayScalar(Night Scale=0.5, Morning Scale=0.8, Evening Scale=0.8)
        ))
       Channel=::NewEnumerator0
      )
     )
     2=&Multiply_LinearColorByFloat(
      Linear Color=&LerpLinearColors(
       StartColor=<rgba 3.8686 0.7083 4 1>
       EndColor=<rgba 0.0086 0 0.5573 0.08>
       LerpFactor=&FN_FrontEndFloatModifier(Gameplay=&Lerp_Float(A=0, B=1, Alpha=0.48), Front End=0.7)
      )
      Float=&MakeFloatFromVector4(
       Vector 4=&MakeVector4_FromFloat(Float=&Multiply_Float(
         A=$System.Timeline.Time
         B=&FNTimeOfDayScalar(Night Scale=0.5, Morning Scale=0.8, Evening Scale=0.8)
        ))
       Channel=::NewEnumerator0
      )
     )
     3=&Multiply_LinearColorByFloat(
      Linear Color=&LerpLinearColors(
       StartColor=<rgba 2.7 1.8223 0.9141 1>
       EndColor=<rgba 0.2491 0.1575 0.0914 0.08>
       LerpFactor=&FN_FrontEndFloatModifier(Gameplay=&Lerp_Float(A=0, B=1, Alpha=0.45), Front End=0.7)
      )
      Float=&MakeFloatFromVector4(
       Vector 4=&MakeVector4_FromFloat(Float=&Multiply_Float(
         A=$System.Timeline.Time
         B=&FNTimeOfDayScalar(Night Scale=0.5, Morning Scale=0.8, Evening Scale=0.8)
        ))
       Channel=::NewEnumerator0
      )
     )
     Index=&MakeIntFromFloat(Float=$User.VariantSwitch, Conversion Mode=::NewEnumerator1)
     Num=::NewEnumerator1
    )) {isSetParameters=true}
  ]
  rendererClasses [&NiagaraRibbonRendererProperties]
  renderers [
   {rendererClass=&NiagaraRibbonRendererProperties}
   {rendererClass=&NiagaraRibbonRendererProperties, rendererIndex=1}
  ]
 }
}
summary {
 systemName=NS_Glider_FoundKelp
 emitters [
  {emitterName=Speedlines, rendererClasses [&NiagaraSpriteRendererProperties]}
  {emitterName=Glow, rendererClasses [&NiagaraSpriteRendererProperties]}
  {emitterName=BladeGlows, rendererClasses [&NiagaraSpriteRendererProperties]}
  {emitterName=Trail002, rendererClasses [&NiagaraRibbonRendererProperties]}
  {emitterName=Blades, rendererClasses [&NiagaraSpriteRendererProperties]}
 ]
 userVariables [
  User.bIsFrontEnd {type=NiagaraBool}
  User.bIsFrontEndPreview {type=NiagaraBool}
  User.bIsFullyDeployed {type=NiagaraBool}
  User.BottomColor {type=LinearColor, description="", defaultValue {struct=LinearColor, value=<rgba 1 1 1 1>}}
  User.ColorRampMult {type=NiagaraFloat, defaultValue=0, description=""}
  User.ColorRampPower {type=NiagaraFloat, defaultValue=0, description=""}
  User.ForwardDot {type=NiagaraFloat, defaultValue=0, description=""}
  User.MainColor {type=LinearColor, description="", defaultValue {struct=LinearColor, value=<rgba 1 1 1 1>}}
  User.RibbonIDOffset {type=NiagaraInt32}
  User.RightDot {type=NiagaraFloat, defaultValue=0, description=""}
  User.TrailLifeTimeMult {type=NiagaraFloat, defaultValue=0.4, description=""}
  User.VariantSwitch {type=NiagaraFloat, defaultValue=1, description=""}
  User.WidthMult {type=NiagaraFloat, defaultValue=2, description=""}
 ]
}
